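# RageBlade_SuperAttack_Cue is a sound cue that plays the RageBlade melee weapon's super attack sound.
# A single wave player for SuperAttack_01 feeds a modulator left at its default settings, with no random selection between waves.
# The cue raises its volume and routes through the melee weapon sound class, small-melee attenuation and player melee concurrency set.
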SoundCue /Game/Sounds/Fort_Weapons/RageBlade/RageBlade_SuperAttack_Cue
root=Modulator_0
nodes [
 {id=Modulator_0, children [WavePlayer_0]}
 {id=WavePlayer_0, properties {soundWaveAssetPtr=&/RageBlade_SuperAttack_01}}
]
properties {
 attenuationSettings=&Fort_Weapon_Melee_Sml_Att
 maxDistance=2800
 soundClassObject=&Fort_WeaponMelee_PSM
 volumeMultiplier=2
 concurrencySet [&Player_Melee_Weapon_Conc]
}
